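; Gameplay ability that grants a transient quest for an urgent telescope quest (Q03 variant) when the owner gains a gameplay tag.
; On activation it casts the owning pawn's controller to the Athena player controller and waits for the configured tag to be added.
; When the tag is added, it grants the configured quest definition through that controller's transient quests component.
; The failed-activation and end events only call their parent functions.
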(event Ability|EventFailedToActivatePassiveAbility
  (params (ActorInfo :type Gameplay%20Ability%20Actor%20Info%20Structure))
  (|CallParentFunction))

(event Ability|EventActivateAbility
  (|CallParentFunction)
  (bind _asfort_player_controller_athena (Utilities|Casting|CastToFortPlayerControllerAthena (Fortnite|GetFortPlayerControllerfromActor (Variables|GAVars|Data|GetOwningPlayerPawn))))
  (Ability|Tasks|WaitGameplayTagAdd (Variables|Default|GetTag)
    (:Added
      (Class|FortControllerComponentTransientQuests|GrantTransientQuest (Class|FortPlayerControllerAthena|GetTransientQuestsComponent _asfort_player_controller_athena) (Variables|Default|GetQuest%20Def%20to%20Grant) 0))))

(event Ability|EventOnEndAbility
  (params (bWasCancelled :type Boolean))
  (|CallParentFunction bWasCancelled))
